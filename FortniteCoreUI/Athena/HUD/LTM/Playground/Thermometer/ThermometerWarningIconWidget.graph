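# Red warning icon widget for the thermometer HUD in the Playground LTM.
# A single centered image, StormIcon, sits in an overlay and shows the WarningIcon texture.
# Anim_dmgON plays the IconDMG_On animation, which snaps the icon from transparent to opaque while scaling it through two quick, diminishing pulses.
# The event graph holds only unconnected default events, so nothing in this widget calls that function.
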WidgetBlueprint /FortniteCoreUI/Athena/HUD/LTM/Playground/Thermometer/ThermometerWarningIconWidget
animations [IconDMG_On {
  binding_count=1
  display_label=""
  length_seconds=3
  object=&self:IconDMG_On
  sequence {bindings [StormIcon {
     id=#1
     possessed_class=Image
     tracks [
      {
       type=MovieScene2DTransformTrack
       sections [{
         type=MovieScene2DTransformSection
         end_frame=59
         start_frame=0
         channels [
          Translation.X {type=MovieSceneScriptingFloatChannel, const [0,1]}
          Translation.Y {type=MovieSceneScriptingFloatChannel, const [0,1]}
          Angle {type=MovieSceneScriptingFloatChannel, const [0,1]}
          Scale.X {type=MovieSceneScriptingFloatChannel, keys [[0,1],[2,1.2],[4,1],[6,1.1],[8,1]]}
          Scale.Y {type=MovieSceneScriptingFloatChannel, keys [[0,1],[2,1.2],[4,1],[6,1.1],[8,1]]}
          Shear.X {type=MovieSceneScriptingFloatChannel, const [0,1]}
          Shear.Y {type=MovieSceneScriptingFloatChannel, const [0,1]}
         ]
        }]
      }
      {
       type=MovieSceneFloatTrack
       display_name=Render Opacity
       sections [{type=MovieSceneFloatSection, end_frame=59, start_frame=0, channels [None {type=MovieSceneScriptingFloatChannel, keys [[0,0],[1,1]]}]}]
      }
     ]
    }]}
 }]
graphs [
 Anim_dmgON {nodes [
   Anim_dmgON() {is=title K2Node_FunctionEntry}
   > PlayAnimation(InAnimation=Get IconDMG_On() -> (IconDMG_On:Widget Animation Object Reference))
  ]}
 EventGraph {nodes [
   Event PreConstruct()
   Event Construct()
   Event Tick()
  ]}
]
variables []
widgets {
 info {inheritedWidgetCount=0, parentClass=&UserWidget, rootWidgetClass=&Overlay, widgetCount=2}
 widgets [
  Overlay_3 {widgetClassPath=&Overlay, widget=Overlay_3}
  StormIcon {
   widgetClassPath=&Image
   bIsVariable=true
   parent=Overlay_3
   widget=StormIcon
   properties {brush {
     bIsDynamicallyLoaded=false
     drawAs=Image
     imageType=NoImage
     mirroring=NoMirror
     resourceName=None
     resourceObject=/Game/Athena/HUD/LTM/Playground/Thermometer/WarningIcon
     tiling=NoTile
     imageSize=<84 84>
     margin=<ltrb 0 0 0 0>
     outlineSettings {
      bUseBrushTransparency=false
      roundingType=HalfHeightRadius
      width=0
      color {colorUseRule=UseColor_Specified, specifiedColor=<rgba 0 0 0 0>}
      cornerRadii=<0 0 0 0>
     }
     tintColor {colorUseRule=UseColor_Specified, specifiedColor=<rgba 1 0.0625 0.0625 1>}
     uVRegion {bIsValid=false, max=<0 0>, min=<0 0>}
    }}
   slot {class=OverlaySlot, properties {horizontalAlignment=HAlign_Center, verticalAlignment=VAlign_Center, padding=<ltrb 7 7 7 7>}}
  }
 ]
}
