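# Translucent UI material for the heatmap icon, filed with the Playground thermometer HUD assets.
# A single sample of the HeatmapIcon_02 texture feeds three simple math nodes: red added to itself, alpha doubled, and RGB scaled up by half.
# The graph shows no links from these nodes to material outputs, so which results reach color and opacity is not visible.
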Material /Game/Athena/HUD/LTM/Playground/Thermometer/M_HeatmapIcon
graph {
 properties {blendMode=BLEND_Translucent, materialDomain=MD_UI, pixelDepthOffsetMode=PDOM_Legacy, usedShadingModels=Default Lit}
 nodes [
  n1 = TextureSample() {texture=&/HeatmapIcon_02}
  Add(A=@n1.R, B=@n1.R)
  Multiply(A=@n1.A) {constB=2}
  Multiply(A=@n1.RGB) {constB=1.5}
 ]
}
